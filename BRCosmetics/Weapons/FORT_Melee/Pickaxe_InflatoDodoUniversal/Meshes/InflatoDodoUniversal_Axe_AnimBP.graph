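# Animation Blueprint for the InflatoDodoUniversal pickaxe mesh, driving its equip animation.
# A boolean blend plays the non-looping equip sequence from an offset start when the pickaxe is both in the front end and on a pawn; otherwise it holds that sequence at time zero.
# The result feeds DefaultSlot, letting montages override the pose.
# The condition flags are not declared locally, and the event graph has no connected logic.
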AnimBlueprint /BRCosmetics/Weapons/FORT_Melee/Pickaxe_InflatoDodoUniversal/Meshes/InflatoDodoUniversal_Axe_AnimBP
target_skeleton=&/InflatoDodoUniversal_Axe_Skeleton
graphs [
 AnimGraph {nodes [
   Input Pose()
   n1 = "Sequence Evaluator (by time) 'InflatoDodoUniversal_equip'"!(ExplicitTime:float=0) -> (Pose:Pose Link Structure) {
    is=title AnimGraphNode_SequenceEvaluator
    sequence=/BRCosmetics/Weapons/FORT_Melee/Pickaxe_InflatoDodoUniversal/Anim/InflatoDodoUniversal_equip
   }
   n2 = Sequence Player 'InflatoDodoUniversal_equip'() {bLoopAnimation=false, sequence=&InflatoDodoUniversal_equip, startPosition=1}
   n3 = AND Boolean(A=Get bIsInFrontEnd(), B=Get bIsOnPawn() -> (bIsOnPawn:bool))
   Output Pose(Result=Slot 'DefaultSlot'(Source=Blend Poses by bool(BlendPose_0=@n2, BlendPose_1=@n1, bActiveValue=@n3) {blendType=Linear})) {layerGroup=None}
  ]}
 EventGraph {nodes [
   TryGetPawnOwner()
   Event BlueprintUpdateAnimation()
  ]}
]
variables []
